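# Post-process style AnimBlueprint for the QuarterClaspZoom male body cosmetic, adding secondary motion to laces, T-shirt sleeves and a chain.
# It copies the parent mesh pose into a Control Rig fed an isFrontend flag.
# Weighted constraints position lace bases, then Anim Dynamics simulates each part, with an additive chain-base offset.
# The event graph updates knee-bend curves, isFrontend, and a chain flag from the face mesh name, which the AnimGraph never visibly reads.
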AnimBlueprint /BRCosmetics/Characters/Player/Male/Medium/Bodies/M_MED_QuarterClaspZoom/Meshes/M_MED_QuarterClaspZoom_AnimBP
target_skeleton=&/M_MED_QuarterClaspZoom_Skeleton
graphs [
 AnimGraph {nodes [
   n1 = Control Rig!(Source=Copy Pose From Mesh() {bUseAttachedParent=true}, Alpha, isFrontend:bool=Get isFrontend() -> (isFrontend:bool)) -> (Pose) {controlRigAssetReference {blueprintRigClass=&/M_MED_QuarterClaspZoom_CtrlRig.M_MED_QuarterClaspZoom_CtrlRig_C}}
   n2 = "Constraint - dyn_laces_out_l_base (PPPPPPPP)"!(
    ComponentPose=Local To Component(LocalPose=@n1)
    Alpha=1
    ConstraintWeights_0=0
    ConstraintWeights_1:float=0.0869
    ConstraintWeights_2:float=0.0094
    ConstraintWeights_3:float=0.5483
    ConstraintWeights_4:float=0.0001
    ConstraintWeights_5:float=0
    ConstraintWeights_6:float=0.3552
    ConstraintWeights_7:float=0
   ) -> (Pose) {
    boneToModify {boneName=dyn_laces_out_l_base}
    constraintSetup [
     {targetBone {boneName=calf_l}}
     {targetBone {boneName=calf_twist_01_l}}
     {targetBone {boneName=calf_twist_02_l}}
     {targetBone {boneName=foot_l}}
     {targetBone {boneName=ball_l}}
     {targetBone {boneName=dyn_foot_l_bk}}
     {targetBone {boneName=dyn_foot_l_fr}}
     {targetBone {boneName=dyn_laces_out_l}}
    ]
    constraintWeights [1,1,1,1,1,1,1,1]
   }
   n3 = "Constraint - dyn_laces_in_l_base (PPPPPPPP)"!(
    ComponentPose=@n2
    Alpha=1
    ConstraintWeights_0=0
    ConstraintWeights_1:float=0.0749
    ConstraintWeights_2:float=0.0127
    ConstraintWeights_3:float=0.374
    ConstraintWeights_4:float=0.0001
    ConstraintWeights_5:float=0
    ConstraintWeights_6:float=0
    ConstraintWeights_7:float=0.5382
   ) -> (Pose) {
    boneToModify {boneName=dyn_laces_in_l_base}
    constraintSetup [
     {targetBone {boneName=calf_l}}
     {targetBone {boneName=calf_twist_01_l}}
     {targetBone {boneName=calf_twist_02_l}}
     {targetBone {boneName=foot_l}}
     {targetBone {boneName=ball_l}}
     {targetBone {boneName=calf_twist_01_r}}
     {targetBone {boneName=dyn_foot_l_bk}}
     {targetBone {boneName=dyn_foot_l_fr}}
    ]
    constraintWeights [1,1,1,1,1,1,1,1]
   }
   n4 = "Constraint - dyn_laces_in_r_base (PPPPPPPP)"!(
    ComponentPose=@n3
    Alpha=1
    ConstraintWeights_0=0
    ConstraintWeights_1:float=0
    ConstraintWeights_2:float=0.0749
    ConstraintWeights_3:float=0.0127
    ConstraintWeights_4:float=0.374
    ConstraintWeights_5:float=0.0001
    ConstraintWeights_6:float=0.5382
    ConstraintWeights_7:float=0
   ) -> (Pose) {
    boneToModify {boneName=dyn_laces_in_r_base}
    constraintSetup [
     {targetBone {boneName=calf_twist_01_l}}
     {targetBone {boneName=calf_r}}
     {targetBone {boneName=calf_twist_01_r}}
     {targetBone {boneName=calf_twist_02_r}}
     {targetBone {boneName=foot_r}}
     {targetBone {boneName=ball_r}}
     {targetBone {boneName=dyn_foot_r_fr}}
     {targetBone {boneName=dyn_foot_r_bk}}
    ]
    constraintWeights [1,1,1,1,1,1,1,1]
   }
   n5 = "Constraint - dyn_laces_out_r_base (PPPPPPPP)"!(
    ComponentPose=@n4
    Alpha=1
    ConstraintWeights_0=0
    ConstraintWeights_1:float=0.0869
    ConstraintWeights_2:float=0.0094
    ConstraintWeights_3:float=0.5483
    ConstraintWeights_4:float=0.0001
    ConstraintWeights_5:float=0.3552
    ConstraintWeights_6:float=0
    ConstraintWeights_7:float=0
   ) -> (Pose) {
    boneToModify {boneName=dyn_laces_out_r_base}
    constraintSetup [
     {targetBone {boneName=calf_r}}
     {targetBone {boneName=calf_twist_01_r}}
     {targetBone {boneName=calf_twist_02_r}}
     {targetBone {boneName=foot_r}}
     {targetBone {boneName=ball_r}}
     {targetBone {boneName=dyn_foot_r_fr}}
     {targetBone {boneName=dyn_laces_out_r}}
     {targetBone {boneName=dyn_foot_r_bk}}
    ]
    constraintWeights [1,1,1,1,1,1,1,1]
   }
   n6 = Anim Dynamics - Bone: dyn_laces_in_l(ComponentPose=@n5, Alpha=1, ExternalForce="0,0,0") {
    angularDampingOverride=0.99
    angularSpringConstant=100
    bAngularSpring=true
    bOverrideAngularDamping=true
    bUseGravityOverride=true
    linearDampingOverride=1
    linearSpringConstant=50
    simulationSpace=RootRelative
    alphaBoolBlend {blendOption=Linear}
    boundBone {boneName=dyn_laces_in_l}
    componentAppliedLinearAccClamp=<5000 5000 5000>
    componentLinearAccScale=<1 1 1>
    componentLinearVelScale=<1 1 1>
    gravityOverride {z=-980}
    physicsBodyDefinitions [{
      boundBone {boneName=dyn_laces_in_l}
      boxExtents=<10 10 10>
      constraintSetup {angularLimitsMax=<5 3 0>, angularLimitsMin=<-5 -30 -10>, angularTarget=<1 0 0>}
      localJointOffset=<0.8 0 0>
     }]
   }
   n7 = Anim Dynamics - Bone: dyn_laces_out_l(ComponentPose=@n6, Alpha=1, ExternalForce="0,0,0") {
    angularDampingOverride=0.99
    angularSpringConstant=100
    bAngularSpring=true
    bOverrideAngularDamping=true
    bUseGravityOverride=true
    linearDampingOverride=1
    linearSpringConstant=50
    simulationSpace=RootRelative
    alphaBoolBlend {blendOption=Linear}
    boundBone {boneName=dyn_laces_out_l}
    componentAppliedLinearAccClamp=<5000 5000 5000>
    componentLinearAccScale=<1 1 1>
    componentLinearVelScale=<1 1 1>
    gravityOverride {z=-980}
    physicsBodyDefinitions [{
      boundBone {boneName=dyn_laces_out_l}
      boxExtents=<10 10 10>
      constraintSetup {angularLimitsMax=<5 3 10>, angularLimitsMin=<-5 -30 0>, angularTarget=<1 0 0>}
      localJointOffset=<0.8 0 0>
     }]
   }
   n8 = Anim Dynamics - Bone: dyn_laces_in_r(ComponentPose=@n7, Alpha=1, ExternalForce="0,0,0") {
    angularDampingOverride=0.99
    angularSpringConstant=100
    bAngularSpring=true
    bOverrideAngularDamping=true
    bUseGravityOverride=true
    linearDampingOverride=1
    linearSpringConstant=50
    simulationSpace=RootRelative
    alphaBoolBlend {blendOption=Linear}
    boundBone {boneName=dyn_laces_in_r}
    componentAppliedLinearAccClamp=<5000 5000 5000>
    componentLinearAccScale=<1 1 1>
    componentLinearVelScale=<1 1 1>
    gravityOverride {z=-980}
    physicsBodyDefinitions [{
      boundBone {boneName=dyn_laces_in_r}
      boxExtents=<10 10 10>
      constraintSetup {angularLimitsMax=<5 3 0>, angularLimitsMin=<-5 -30 -10>, angularTarget=<1 0 0>}
      localJointOffset=<0.8 0 0>
     }]
   }
   n9 = Anim Dynamics - Bone: dyn_laces_out_r(ComponentPose=@n8, Alpha=1, ExternalForce="0,0,0") {
    angularDampingOverride=0.99
    angularSpringConstant=100
    bAngularSpring=true
    bOverrideAngularDamping=true
    bUseGravityOverride=true
    linearDampingOverride=1
    linearSpringConstant=50
    simulationSpace=RootRelative
    alphaBoolBlend {blendOption=Linear}
    boundBone {boneName=dyn_laces_out_r}
    componentAppliedLinearAccClamp=<5000 5000 5000>
    componentLinearAccScale=<1 1 1>
    componentLinearVelScale=<1 1 1>
    gravityOverride {z=-980}
    physicsBodyDefinitions [{
      boundBone {boneName=dyn_laces_out_r}
      boxExtents=<10 10 10>
      constraintSetup {angularLimitsMax=<5 3 10>, angularLimitsMin=<-5 -30 0>, angularTarget=<1 0 0>}
      localJointOffset=<0.8 0 0>
     }]
   }
   n10 = Anim Dynamics - Bone: upperarm_tshirt_l(ComponentPose=@n9, Alpha=1, ExternalForce="0,0,0") {
    angularDampingOverride=0.9999
    angularSpringConstant=200
    bAngularSpring=true
    bLinearSpring=true
    bOverrideAngularDamping=true
    bOverrideLinearDamping=true
    bUseGravityOverride=true
    linearDampingOverride=1
    linearSpringConstant=50
    alphaBoolBlend {blendOption=Linear}
    boundBone {boneName=upperarm_tshirt_l}
    componentAppliedLinearAccClamp=<3000 3000 3000>
    componentLinearAccScale=<0.4 0.4 0.4>
    componentLinearVelScale=<0.4 0.4 0.4>
    gravityOverride {z=-980}
    physicsBodyDefinitions [{
      boundBone {boneName=upperarm_tshirt_l}
      boxExtents=<10 10 10>
      constraintSetup {angularLimitsMax=<5 5 5>, angularLimitsMin=<-5 -5 -5>, angularTarget=<1 0 0>, linearAxesMin=<-2 0 0>}
      localJointOffset=<0.4 0 0>
     }]
   }
   n11 = Anim Dynamics - Bone: upperarm_tshirt_r(ComponentPose=@n10, Alpha=1, ExternalForce="0,0,0") {
    angularDampingOverride=0.9999
    angularSpringConstant=200
    bAngularSpring=true
    bLinearSpring=true
    bOverrideAngularDamping=true
    bOverrideLinearDamping=true
    bUseGravityOverride=true
    linearDampingOverride=1
    linearSpringConstant=50
    alphaBoolBlend {blendOption=Linear}
    boundBone {boneName=upperarm_tshirt_r}
    componentAppliedLinearAccClamp=<3000 3000 3000>
    componentLinearAccScale=<0.4 0.4 0.4>
    componentLinearVelScale=<0.4 0.4 0.4>
    gravityOverride {z=-980}
    physicsBodyDefinitions [{
      boundBone {boneName=upperarm_tshirt_r}
      boxExtents=<10 10 10>
      constraintSetup {angularLimitsMax=<5 5 5>, angularLimitsMin=<-5 -5 -5>, angularTarget=<1 0 0>, linearAxesMax=<0 0 2>}
      localJointOffset=<0.4 0 0>
     }]
   }
   n12 = "Transform (Modify) Bone - Bone: dyn_chain_base"!(ComponentPose=@n11, bAlphaBoolEnabled:bool=True, Translation="-0.144,0.663,1.078", Rotation="-0.000,0.000,5.153", Scale="1,1,1") -> (Pose) {
    alphaInputType=Bool
    rotationMode=BMM_Additive
    translationMode=BMM_Additive
    boneToModify {boneName=dyn_chain_base}
    rotation {roll=5.153}
    translation=<-0.144 0.663 1.078>
   }
   n13 = Anim Dynamics - Bone: dyn_chain!(ComponentPose=@n12, bAlphaBoolEnabled:bool=True, ExternalForce="0,0,0") -> (Pose) {
    alphaInputType=Bool
    angularDampingOverride=0.999
    angularSpringConstant=150
    bAngularSpring=true
    bOverrideAngularDamping=true
    bOverrideLinearDamping=true
    bUseGravityOverride=true
    bUsePlanarLimit=false
    linearDampingOverride=0.999
    linearSpringConstant=300
    alphaBoolBlend {blendOption=Linear}
    boundBone {boneName=dyn_chain}
    componentAppliedLinearAccClamp=<3000 3000 3000>
    componentLinearAccScale=<0.5 0.5 0.5>
    componentLinearVelScale=<0.5 0.5 0.5>
    gravityOverride {z=-980}
    physicsBodyDefinitions [{
      boundBone {boneName=dyn_chain}
      boxExtents=<10 10 10>
      constraintSetup {angularLimitsMax=<0 0 20>, angularLimitsMin=<0 -30 -20>, angularTarget=<1 0 0>, linearAxesMax=<0 1 0.5>, linearAxesMin=<0 -1 0>}
      localJointOffset=<0.5 0 0>
     }]
   }
   Output Pose(Result=Component To Local(ComponentPose=@n13)) {layerGroup=None}
  ]}
 EventGraph {nodes [
   Event BlueprintUpdateAnimation()
   > n1 = Sequence()
   n2 = GetCurveValue(CurveName=knee_bend_l)
   > Set Knee_bend_l(Knee_bend_l:float=@n2) -> (Output_Get:float)
   > Set Knee_bend_l_inv(Knee_bend_l_inv:float=Remap Float(Input=@n2, OldRange=(X=0,Y=1), NewRange=(X=1,Y=0), Clamp=true) -> (Output=~)) -> (Output_Get:float)
   n3 = GetCurveValue(CurveName=knee_bend_r)
   > Set Knee_bend_r(Knee_bend_r:float=@n3) -> (Output_Get:float)
   > Set Knee_bend_r_inv(Knee_bend_r_inv:float=Remap Float(Input=@n3, OldRange=(X=0,Y=1), NewRange=(X=1,Y=0), Clamp=true) -> (Output=~)) -> (Output_Get:float)
   n4 = Get bIsPlayingEmote()
   > Set isFrontend(isFrontend:bool=AND Boolean(A=Get bIsInFrontEnd(), B=NOT Boolean(A=@n4))) -> (Output_Get:bool)
   n5 = GetDisplayName(Object=GetSkeletalMeshForPartType(self=Get As Fort Player Pawn Athena() -> (As Fort Player Pawn Athena:Fort Player Pawn Athena Object Reference), PartType=Face))
   n6 = Contains(SearchIn=@n5, Substring=QuarterClasp_Pecan)
   n7 = Contains(SearchIn=@n5, Substring=QuarterClasp_Birch)
   Set isChainDynActive(execute=@n1.then_1, isChainDynActive:bool=NOT Boolean(A=OR Boolean(A=@n7, B=@n6))) -> (Output_Get:bool)
   Event OnCharacterCustomizationLoaded()
   > n8 = Cast To FortPlayerPawnAthena(Object=TryGetPawnOwner())
   > Set As Fort Player Pawn Athena(As Fort Player Pawn Athena:Fort Player Pawn Athena Object Reference=@n8) -> (Output_Get:Fort Player Pawn Athena Object Reference)
  ]}
]
variables [
 As Fort Player Pawn Athena {type=FortPlayerPawnAthena}
 isChainDynActive {type=boolean}
 isFrontend {type=boolean}
 Knee_bend_l {type=number}
 Knee_bend_l_inv {type=number}
 Knee_bend_r {type=number}
 Knee_bend_r_inv {type=number}
]
