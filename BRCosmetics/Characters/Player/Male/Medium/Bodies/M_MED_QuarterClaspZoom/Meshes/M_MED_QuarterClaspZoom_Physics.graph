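# Physics asset for the QuarterClaspZoom medium male player body cosmetic, set up around a dangling chain accessory (dyn_chain).
# One constraint joins the dyn_chain capsule to a tiny dyn_chain_base anchor capsule, with uneven swing limits and a tight twist limit.
# Clavicle capsules, plus a capsule and thin tilted box on spine_05, form upper-torso collision shapes around the chain.
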PhysicsAsset /BRCosmetics/Characters/Player/Male/Medium/Bodies/M_MED_QuarterClaspZoom/Meshes/M_MED_QuarterClaspZoom_Physics
bodies {
 clavicle_l [{
   shapeType=Capsule
   length=10.1175
   radius=9.6403
   shapeName=clavicle_l_capsule
   center=<10.2894 -0.6985 -1.4078>
   rotation=<rot 29.3611 78.6766 128.9487>
  }]
 clavicle_r [{
   shapeType=Capsule
   length=10.1175
   radius=9.6403
   shapeName=clavicle_r_capsule
   center=<-10.3069 0.611 1.3576>
   rotation=<rot 29.3611 78.6766 -51.0513>
  }]
 dyn_chain [{
   shapeType=Capsule
   length=16.8917
   radius=7.2717
   shapeName=dyn_chain_capsule
   center=<3.4816 -0.2906 6.9276>
   rotation=<rot 0 -2.9741 -89.993>
  }]
 dyn_chain_base [{shapeType=Capsule, length=0.505, radius=0.505, shapeName=dyn_chain_base_capsule, rotation=<rot 0 -90 0>}]
 spine_05 [
  {shapeType=Capsule, length=15.5367, radius=7.5739, shapeName=spine_05_capsule, center=<15.9166 -2.0169 0>, rotation=<rot 90 19.9997 0>}
  {shapeType=Box, extentX=0.8081, extentY=22.6356, extentZ=29.6171, shapeName=spine_05_box, center=<7.0453 6.657 0>, rotation=<rot 0 49.9544 0>}
 ]
}
constraints [{bone1Name=dyn_chain, bone2Name=dyn_chain_base, swing1LimitDegrees=35, swing2LimitDegrees=15, twistLimitDegrees=5}]
